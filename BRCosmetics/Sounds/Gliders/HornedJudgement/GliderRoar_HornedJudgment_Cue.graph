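# Sound cue for the roar of the Horned Judgement glider cosmetic.
# It picks one of three roar wave variations at equal weight and passes it through a modulator that slightly lowers volume with pitch held fixed.
# It uses the close-range generic in-air glider attenuation and is routed to the glider landing sound class.
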SoundCue /BRCosmetics/Sounds/Gliders/HornedJudgement/GliderRoar_HornedJudgment_Cue
root=Modulator_0
nodes [
 {id=Modulator_0, children [Random_0], properties {pitchMax=1, pitchMin=1, volumeMax=0.8, volumeMin=0.8}}
 {id=Random_0, children [WavePlayer_0,WavePlayer_1,WavePlayer_2], properties {weights [1,1,1]}}
 {id=WavePlayer_0, properties {soundWaveAssetPtr=&/GliderRoar_HornedJudgment_01}}
 {id=WavePlayer_1, properties {soundWaveAssetPtr=&/GliderRoar_HornedJudgment_02}}
 {id=WavePlayer_2, properties {soundWaveAssetPtr=&/GliderRoar_HornedJudgment_03}}
]
properties {
 attenuationSettings=/Game/Sounds/Attenuation/Fort_Player_Glider_GenericInAir_Close_Athena_Att
 maxDistance=9000
 soundClassObject=&Fort_Gliders_Land
}
