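# Sound cue for the wing flap sound of the Horned Judgement glider cosmetic.
# A Random node picks one of six equally weighted flap waves, then a Modulator applies a slight, fixed volume reduction with no volume variation.
# It uses the shared close-range in-air player glider attenuation and is assigned to the glider landing sound class.
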SoundCue /BRCosmetics/Sounds/Gliders/HornedJudgement/GliderWingFlap_HornedJudgment_Cue
root=Modulator_0
nodes [
 {id=Modulator_0, children [Random_0], properties {volumeMax=0.9, volumeMin=0.9}}
 {id=Random_0, children [WavePlayer_0,WavePlayer_1,WavePlayer_2,WavePlayer_3,WavePlayer_4,WavePlayer_5], properties {weights [1,1,1,1,1,1]}}
 {id=WavePlayer_0, properties {soundWaveAssetPtr=&/GliderWingFlap_HornedJudgment_06}}
 {id=WavePlayer_1, properties {soundWaveAssetPtr=&/GliderWingFlap_HornedJudgment_05}}
 {id=WavePlayer_2, properties {soundWaveAssetPtr=&/GliderWingFlap_HornedJudgment_04}}
 {id=WavePlayer_3, properties {soundWaveAssetPtr=&/GliderWingFlap_HornedJudgment_03}}
 {id=WavePlayer_4, properties {soundWaveAssetPtr=&/GliderWingFlap_HornedJudgment_02}}
 {id=WavePlayer_5, properties {soundWaveAssetPtr=&/GliderWingFlap_HornedJudgment_01}}
]
properties {
 attenuationSettings=/Game/Sounds/Attenuation/Fort_Player_Glider_GenericInAir_Close_Athena_Att
 maxDistance=9000
 soundClassObject=&Fort_Gliders_Land
}
